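# Translucent, unlit, two-sided Niagara material for the film-ribbon contrail of the Tandem skydiving cosmetic.
# Color comes from a curve atlas sampled along U, modulated by noise and particle color, with desaturation and simple-glow controls.
# Opacity combines a panning film texture with noise or ribbon masking, end fades and camera depth fade.
# Optional world position offset adds two U-ramped sine waves to SplineThicken, with dynamic parameters driving width, offset and motion.
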Material /BRCosmetics/Athena/Effects/Particles/Skydiving/NS_Contrail_Tandem/Materials/M_Contrail_Tandem_Film
graph {
 expressions {
  MaterialExpressionAdd_0 {
   class=MaterialExpressionAdd
   inputs [
    A {from=MaterialExpressionTextureCoordinate_1}
    B {from=MaterialExpressionComponentMask_1}
   ]
  }
  MaterialExpressionAdd_1 {
   class=MaterialExpressionAdd
   inputs [
    A {from=MaterialExpressionLinearInterpolate_5}
    B {from=MaterialExpressionMultiply_9}
   ]
  }
  MaterialExpressionAdd_10 {
   class=MaterialExpressionAdd
   inputs [
    A {from=MaterialExpressionPower_2}
    B {from=MaterialExpressionStaticSwitchParameter_1}
   ]
  }
  MaterialExpressionAdd_2 {
   class=MaterialExpressionAdd
   inputs [
    A {from=MaterialExpressionComponentMask_2}
    B {from=MaterialExpressionMultiply_5}
   ]
  }
  MaterialExpressionAdd_3 {
   class=MaterialExpressionAdd
   inputs [
    A {from=MaterialExpressionComponentMask_3}
    B {from=MaterialExpressionMultiply_7}
   ]
  }
  MaterialExpressionAdd_4 {
   class=MaterialExpressionAdd
   inputs [
    A {from=MaterialExpressionMultiply_6}
    B {from=MaterialExpressionMultiply_8}
   ]
  }
  MaterialExpressionAdd_5 {
   class=MaterialExpressionAdd
   inputs [
    A {from=MaterialExpressionTextureCoordinate_7}
    B {from=MaterialExpressionConvert_0}
   ]
  }
  MaterialExpressionAdd_6 {
   class=MaterialExpressionAdd
   inputs [
    A {from=MaterialExpressionTextureCoordinate_3}
    B {from=MaterialExpressionConvert_1}
   ]
  }
  MaterialExpressionAdd_7 {
   class=MaterialExpressionAdd
   inputs [
    A {from=MaterialExpressionTextureCoordinate_4}
    B {from=MaterialExpressionConvert_2}
   ]
  }
  MaterialExpressionAdd_8 {
   class=MaterialExpressionAdd
   inputs [
    A {from=MaterialExpressionTextureCoordinate_0}
    B {from=MaterialExpressionComponentMask_7}
   ]
  }
  MaterialExpressionAdd_9 {
   class=MaterialExpressionAdd
   inputs [
    A {from=MaterialExpressionOneMinus_0}
    B {from=MaterialExpressionStaticSwitchParameter_0}
   ]
  }
  MaterialExpressionComponentMask_0 {class=MaterialExpressionComponentMask, inputs [None {from=MaterialExpressionAdd_8}], properties {r=true}}
  MaterialExpressionComponentMask_1 {class=MaterialExpressionComponentMask, inputs [None {from=MaterialExpressionVectorParameter_0, output=RGB}], properties {g=true, r=true}}
  MaterialExpressionComponentMask_2 {class=MaterialExpressionComponentMask, inputs [None {from=MaterialExpressionAdd_6}], properties {r=true}}
  MaterialExpressionComponentMask_3 {class=MaterialExpressionComponentMask, inputs [None {from=MaterialExpressionAdd_7}], properties {r=true}}
  MaterialExpressionComponentMask_4 {class=MaterialExpressionComponentMask, inputs [None {from=MaterialExpressionTextureCoordinate_5}], properties {r=true}}
  MaterialExpressionComponentMask_5 {class=MaterialExpressionComponentMask, inputs [None {from=MaterialExpressionTextureCoordinate_6}], properties {r=true}}
  MaterialExpressionComponentMask_6 {class=MaterialExpressionComponentMask, inputs [None {from=MaterialExpressionVectorParameter_3, output=RGB}], properties {g=true, r=true}}
  MaterialExpressionComponentMask_7 {class=MaterialExpressionComponentMask, inputs [None {from=MaterialExpressionVectorParameter_4, output=RGB}], properties {g=true, r=true}}
  MaterialExpressionComponentMask_8 {class=MaterialExpressionComponentMask, inputs [None {from=MaterialExpressionTextureCoordinate_8}], properties {g=true}}
  MaterialExpressionConstant_0 {class=MaterialExpressionConstant, properties {r=-0.8}}
  MaterialExpressionConstant_1 {class=MaterialExpressionConstant}
  MaterialExpressionConstant_2 {class=MaterialExpressionConstant}
  MaterialExpressionConstant_3 {class=MaterialExpressionConstant}
  MaterialExpressionConvert_0 {
   class=MaterialExpressionConvert
   inputs [
    None {from=MaterialExpressionDynamicParameter_1, output=U offset}
    None {from=""}
   ]
   properties {
    nodeName=Make Vector2
    convertInputs [
     {
      type=Scalar
      expressionInput {
       expression=&self:MaterialExpressionDynamicParameter_1
       inputName=None
       mask=1
       maskA=0
       maskB=1
       maskG=0
       maskR=0
       outputIndex=2
      }
     }
     {type=Scalar, expressionInput {expression=null, inputName=None, mask=0, maskA=0, maskB=0, maskG=0, maskR=0, outputIndex=0}}
    ]
    convertOutputs [{type=Vector2}]
   }
  }
  MaterialExpressionConvert_1 {
   class=MaterialExpressionConvert
   inputs [
    None {from=MaterialExpressionDynamicParameter_2, output=U offset}
    None {from=""}
   ]
   properties {
    nodeName=Make Vector2
    convertInputs [
     {
      type=Scalar
      expressionInput {
       expression=&self:MaterialExpressionDynamicParameter_2
       inputName=None
       mask=1
       maskA=0
       maskB=1
       maskG=0
       maskR=0
       outputIndex=2
      }
     }
     {type=Scalar, expressionInput {expression=null, inputName=None, mask=0, maskA=0, maskB=0, maskG=0, maskR=0, outputIndex=0}}
    ]
    convertOutputs [{type=Vector2}]
   }
  }
  MaterialExpressionConvert_2 {
   class=MaterialExpressionConvert
   inputs [
    None {from=MaterialExpressionDynamicParameter_3, output=U offset}
    None {from=""}
   ]
   properties {
    nodeName=Make Vector2
    convertInputs [
     {
      type=Scalar
      expressionInput {
       expression=&self:MaterialExpressionDynamicParameter_3
       inputName=None
       mask=1
       maskA=0
       maskB=1
       maskG=0
       maskR=0
       outputIndex=2
      }
     }
     {type=Scalar, expressionInput {expression=null, inputName=None, mask=0, maskA=0, maskB=0, maskG=0, maskR=0, outputIndex=0}}
    ]
    convertOutputs [{type=Vector2}]
   }
  }
  MaterialExpressionCurveAtlasRowParameter_0 {
   class=MaterialExpressionCurveAtlasRowParameter
   inputs [CurveTime {from=MaterialExpressionComponentMask_0, output=""}]
   properties {
    atlas=&/CAtlas_Contrail_Tandem
    curve=&/CV_Contrail_TandemColor
    parameterName=Color
   }
  }
  MaterialExpressionDesaturation_1 {
   class=MaterialExpressionDesaturation
   inputs [
    None {from=MaterialExpressionCurveAtlasRowParameter_0, output=""}
    Fraction {from=MaterialExpressionScalarParameter_5, output=""}
   ]
   properties {}
  }
  MaterialExpressionDynamicParameter_0 {
   class=MaterialExpressionDynamicParameter
   inputs []
   properties {defaultValue=<rgba 1 20 0 1>, paramNames [WPOPanner, WidthTip, U offset, Pannertime]}
  }
  MaterialExpressionDynamicParameter_1 {
   class=MaterialExpressionDynamicParameter
   inputs []
   properties {defaultValue=<rgba 1 20 0 1>, paramNames [WPOPanner, WidthTip, U offset, Pannertime]}
  }
  MaterialExpressionDynamicParameter_2 {
   class=MaterialExpressionDynamicParameter
   inputs []
   properties {defaultValue=<rgba 1 20 0 1>, paramNames [WPOPanner, WidthTip, U offset, Pannertime]}
  }
  MaterialExpressionDynamicParameter_3 {
   class=MaterialExpressionDynamicParameter
   inputs []
   properties {defaultValue=<rgba 1 20 0 1>, paramNames [WPOPanner, WidthTip, U offset, Pannertime]}
  }
  MaterialExpressionDynamicParameter_4 {
   class=MaterialExpressionDynamicParameter
   inputs []
   properties {defaultValue=<rgba 1 20 0 1>, paramNames [WPOPanner, WidthTip, U offset, Pannertime]}
  }
  MaterialExpressionDynamicParameter_5 {
   class=MaterialExpressionDynamicParameter
   inputs []
   properties {parameterIndex=1, defaultValue=<rgba 1 20 0 1>, paramNames [Edge1Fade, Edge2Fade, "", ""]}
  }
  MaterialExpressionDynamicParameter_6 {
   class=MaterialExpressionDynamicParameter
   inputs []
   properties {defaultValue=<rgba 1 20 0 1>, paramNames [WPOPanner, WidthTip, U offset, Pannertime]}
  }
  MaterialExpressionLinearInterpolate_0 {
   class=MaterialExpressionLinearInterpolate
   inputs [
    A {from=MaterialExpressionScalarParameter_6}
    B {from=""}
    Alpha {from=MaterialExpressionTextureSample_1, output=R}
   ]
   properties {constA=0.3, constB=2}
  }
  MaterialExpressionLinearInterpolate_1 {
   class=MaterialExpressionLinearInterpolate
   inputs [
    A {from=MaterialExpressionSaturate_0}
    B {from=MaterialExpressionLinearInterpolate_4}
    Alpha {from=MaterialExpressionScalarParameter_3}
   ]
  }
  MaterialExpressionLinearInterpolate_2 {
   class=MaterialExpressionLinearInterpolate
   inputs [
    A {from=MaterialExpressionMultiply_1}
    B {from=MaterialExpressionMultiply_13}
    Alpha {from=MaterialExpressionScalarParameter_4}
   ]
  }
  MaterialExpressionLinearInterpolate_3 {
   class=MaterialExpressionLinearInterpolate
   inputs [
    A {from=""}
    B {from=""}
    Alpha {from=MaterialExpressionOneMinus_1}
   ]
   properties {constA=0.5}
  }
  MaterialExpressionLinearInterpolate_4 {
   class=MaterialExpressionLinearInterpolate
   inputs [
    A {from=MaterialExpressionStaticComponentMaskParameter_0}
    B {from=MaterialExpressionMultiply_20}
    Alpha {from=MaterialExpressionScalarParameter_7}
   ]
  }
  MaterialExpressionLinearInterpolate_5 {
   class=MaterialExpressionLinearInterpolate
   inputs [
    A {from=MaterialExpressionMaterialFunctionCall_0, output=WorldPositionOffset}
    B {from=MaterialExpressionConstant_3}
    Alpha {from=MaterialExpressionScalarParameter_8}
   ]
  }
  MaterialExpressionMaterialFunctionCall_0 {
   class=MaterialExpressionMaterialFunctionCall
   inputs [
    WidthBase {from=MaterialExpressionMultiply_18}
    WidthTip {from=MaterialExpressionDynamicParameter_0, output=WidthTip}
    WorldPosition {from=""}
    UVs for Projection {from=""}
    UVs for Thickness {from=""}
    UVs for texturing {from=""}
    Expand U or V UV Channel {from=MaterialExpressionStaticBool_0}
    DeriveNormalZ {from=""}
    AngleCorrectedNormals {from=""}
    AdditionalNormal {from=""}
    FlattenPixelNormal {from=""}
    UVs For NormalShading {from=""}
   ]
   properties {materialFunction=&SplineThicken}
  }
  MaterialExpressionMaterialFunctionCall_1 {
   class=MaterialExpressionMaterialFunctionCall
   inputs [Emissive {from=MaterialExpressionLinearInterpolate_2}]
   properties {materialFunction=&MF_FX_EmissiveAdjustments}
  }
  MaterialExpressionMaterialFunctionCall_3 {
   class=MaterialExpressionMaterialFunctionCall
   inputs [
    Fade Length {from=MaterialExpressionScalarParameter_0}
    Fade Offset {from=MaterialExpressionScalarParameter_1}
    For Vertex Shader {from=""}
   ]
   properties {materialFunction=&CameraDepthFade}
  }
  MaterialExpressionMultiply_0 {
   class=MaterialExpressionMultiply
   inputs [
    A {from=MaterialExpressionParticleColor_0, output=A}
    B {from=MaterialExpressionLinearInterpolate_1}
   ]
  }
  MaterialExpressionMultiply_1 {
   class=MaterialExpressionMultiply
   inputs [
    A {from=MaterialExpressionMultiply_2}
    B {from=MaterialExpressionParticleColor_0, output=RGB}
   ]
   properties {constA=1}
  }
  MaterialExpressionMultiply_10 {
   class=MaterialExpressionMultiply
   inputs [
    A {from=MaterialExpressionComponentMask_4}
    B {from=""}
   ]
   properties {constB=3}
  }
  MaterialExpressionMultiply_11 {
   class=MaterialExpressionMultiply
   inputs [
    A {from=MaterialExpressionMultiply_0}
    B {from=MaterialExpressionMultiply_14}
   ]
  }
  MaterialExpressionMultiply_12 {
   class=MaterialExpressionMultiply
   inputs [
    A {from=MaterialExpressionAdd_9}
    B {from=""}
   ]
   properties {constB=3}
  }
  MaterialExpressionMultiply_13 {
   class=MaterialExpressionMultiply
   inputs [
    A {from=MaterialExpressionParticleColor_1, output=RGB}
    B {from=MaterialExpressionDesaturation_1}
   ]
  }
  MaterialExpressionMultiply_14 {
   class=MaterialExpressionMultiply
   inputs [
    A {from=MaterialExpressionSaturate_2}
    B {from=MaterialExpressionSaturate_3}
   ]
  }
  MaterialExpressionMultiply_15 {
   class=MaterialExpressionMultiply
   inputs [
    A {from=MaterialExpressionAdd_10}
    B {from=""}
   ]
   properties {constB=15}
  }
  MaterialExpressionMultiply_16 {
   class=MaterialExpressionMultiply
   inputs [
    A {from=MaterialExpressionMultiply_11}
    B {from=MaterialExpressionMaterialFunctionCall_3, output=Result}
   ]
  }
  MaterialExpressionMultiply_17 {
   class=MaterialExpressionMultiply
   inputs [
    A {from=MaterialExpressionConstant_0}
    B {from=MaterialExpressionDynamicParameter_6, output=WPOPanner}
   ]
  }
  MaterialExpressionMultiply_18 {
   class=MaterialExpressionMultiply
   inputs [
    A {from=MaterialExpressionDynamicParameter_0, output=WidthTip}
    B {from=""}
   ]
   properties {constB=2}
  }
  MaterialExpressionMultiply_19 {
   class=MaterialExpressionMultiply
   inputs [
    A {from=MaterialExpressionTime_3}
    B {from=MaterialExpressionDynamicParameter_4, output=Pannertime}
   ]
  }
  MaterialExpressionMultiply_2 {
   class=MaterialExpressionMultiply
   inputs [
    A {from=MaterialExpressionLinearInterpolate_0}
    B {from=MaterialExpressionDesaturation_1}
   ]
  }
  MaterialExpressionMultiply_20 {
   class=MaterialExpressionMultiply
   inputs [
    A {from=MaterialExpressionLinearInterpolate_3}
    B {from=MaterialExpressionSaturate_4}
   ]
  }
  MaterialExpressionMultiply_21 {
   class=MaterialExpressionMultiply
   inputs [
    A {from=MaterialExpressionPower_3}
    B {from=""}
   ]
   properties {constB=50}
  }
  MaterialExpressionMultiply_3 {
   class=MaterialExpressionMultiply
   inputs [
    A {from=MaterialExpressionStaticComponentMaskParameter_0}
    B {from=MaterialExpressionTextureSample_2, output=R}
   ]
  }
  MaterialExpressionMultiply_4 {
   class=MaterialExpressionMultiply
   inputs [
    A {from=MaterialExpressionMultiply_3}
    B {from=""}
   ]
   properties {constB=2}
  }
  MaterialExpressionMultiply_5 {
   class=MaterialExpressionMultiply
   inputs [
    A {from=MaterialExpressionTime_0}
    B {from=MaterialExpressionMultiply_17}
   ]
   properties {constA=1, constB=-1}
  }
  MaterialExpressionMultiply_6 {
   class=MaterialExpressionMultiply
   inputs [
    A {from=MaterialExpressionTransform_0}
    B {from=MaterialExpressionSine_0}
   ]
  }
  MaterialExpressionMultiply_7 {
   class=MaterialExpressionMultiply
   inputs [
    A {from=MaterialExpressionTime_1}
    B {from=MaterialExpressionMultiply_17}
   ]
   properties {constA=1, constB=-1}
  }
  MaterialExpressionMultiply_8 {
   class=MaterialExpressionMultiply
   inputs [
    A {from=MaterialExpressionTransform_1}
    B {from=MaterialExpressionSine_1}
   ]
  }
  MaterialExpressionMultiply_9 {
   class=MaterialExpressionMultiply
   inputs [
    A {from=MaterialExpressionAdd_4}
    B {from=MaterialExpressionSaturate_1}
   ]
  }
  MaterialExpressionOneMinus_0 {class=MaterialExpressionOneMinus, inputs [None {from=MaterialExpressionComponentMask_5}]}
  MaterialExpressionOneMinus_1 {class=MaterialExpressionOneMinus, inputs [None {from=MaterialExpressionSine_2}]}
  MaterialExpressionPanner_0 {
   class=MaterialExpressionPanner
   inputs [
    Coordinate {from=MaterialExpressionAdd_5}
    Time {from=MaterialExpressionMultiply_19}
    Speed {from=MaterialExpressionComponentMask_6}
   ]
   properties {speedX=-0.3}
  }
  MaterialExpressionParticleColor_0 {class=MaterialExpressionParticleColor, inputs [], properties {}}
  MaterialExpressionParticleColor_1 {class=MaterialExpressionParticleColor, inputs [], properties {}}
  MaterialExpressionPower_2 {
   class=MaterialExpressionPower
   inputs [
    Base {from=MaterialExpressionComponentMask_5}
    Exp {from=MaterialExpressionScalarParameter_2}
   ]
   properties {constExponent=2.5}
  }
  MaterialExpressionPower_3 {
   class=MaterialExpressionPower
   inputs [
    Base {from=MaterialExpressionSine_2}
    Exp {from=""}
   ]
   properties {constExponent=5}
  }
  MaterialExpressionSaturate_0 {class=MaterialExpressionSaturate, inputs [None {from=MaterialExpressionMultiply_4}]}
  MaterialExpressionSaturate_1 {class=MaterialExpressionSaturate, inputs [None {from=MaterialExpressionMultiply_10}]}
  MaterialExpressionSaturate_2 {class=MaterialExpressionSaturate, inputs [None {from=MaterialExpressionMultiply_12}]}
  MaterialExpressionSaturate_3 {class=MaterialExpressionSaturate, inputs [None {from=MaterialExpressionMultiply_15}]}
  MaterialExpressionSaturate_4 {class=MaterialExpressionSaturate, inputs [None {from=MaterialExpressionMultiply_21}]}
  MaterialExpressionScalarParameter_0 {class=MaterialExpressionScalarParameter, properties {defaultValue=100, parameterName="Fade Length (S)"}}
  MaterialExpressionScalarParameter_1 {class=MaterialExpressionScalarParameter, properties {defaultValue=40, parameterName="Fade Offset (S)"}}
  MaterialExpressionScalarParameter_2 {class=MaterialExpressionScalarParameter, properties {defaultValue=2.5, parameterName=Head Fade Exp, sliderMax=10}}
  MaterialExpressionScalarParameter_3 {class=MaterialExpressionScalarParameter, properties {parameterName=UseSimpleGlow}}
  MaterialExpressionScalarParameter_4 {class=MaterialExpressionScalarParameter, properties {parameterName=UseSimpleGlow}}
  MaterialExpressionScalarParameter_5 {class=MaterialExpressionScalarParameter, properties {parameterName=Des}}
  MaterialExpressionScalarParameter_6 {class=MaterialExpressionScalarParameter, properties {defaultValue=0.3, parameterName=NoiseLerpMin}}
  MaterialExpressionScalarParameter_7 {class=MaterialExpressionScalarParameter, properties {parameterName=UseRibbonMask}}
  MaterialExpressionScalarParameter_8 {class=MaterialExpressionScalarParameter, properties {parameterName=Use_SplineThicken}}
  MaterialExpressionSine_0 {class=MaterialExpressionSine, inputs [None {from=MaterialExpressionAdd_2}]}
  MaterialExpressionSine_1 {class=MaterialExpressionSine, inputs [None {from=MaterialExpressionAdd_3}], properties {period=1.1}}
  MaterialExpressionSine_2 {class=MaterialExpressionSine, inputs [None {from=MaterialExpressionComponentMask_8}], properties {period=2}}
  MaterialExpressionStaticBool_0 {class=MaterialExpressionStaticBool}
  MaterialExpressionStaticComponentMaskParameter_0 {
   class=MaterialExpressionStaticComponentMaskParameter
   inputs [None {from=MaterialExpressionTextureSampleParameter2D_0, output=RGB}]
   properties {defaultR=true, parameterName=RGBMask}
  }
  MaterialExpressionStaticSwitchParameter_0 {
   class=MaterialExpressionStaticSwitchParameter
   inputs [
    True {from=MaterialExpressionDynamicParameter_5, output=Edge1Fade}
    False {from=MaterialExpressionConstant_1, output=""}
   ]
   properties {parameterName=UseEdgeFadeMasks}
  }
  MaterialExpressionStaticSwitchParameter_1 {
   class=MaterialExpressionStaticSwitchParameter
   inputs [
    True {from=MaterialExpressionDynamicParameter_5, output=Edge2Fade}
    False {from=MaterialExpressionConstant_1, output=""}
   ]
   properties {parameterName=UseEdgeFadeMasks}
  }
  MaterialExpressionStaticSwitchParameter_2 {
   class=MaterialExpressionStaticSwitchParameter
   inputs [
    True {from=MaterialExpressionAdd_1, output=""}
    False {from=MaterialExpressionConstant_2, output=""}
   ]
   properties {defaultValue=true, parameterName=UseWPO}
  }
  MaterialExpressionTextureCoordinate_0 {class=MaterialExpressionTextureCoordinate}
  MaterialExpressionTextureCoordinate_1 {class=MaterialExpressionTextureCoordinate, properties {vTiling=0.5}}
  MaterialExpressionTextureCoordinate_2 {class=MaterialExpressionTextureCoordinate, properties {vTiling=0.7}}
  MaterialExpressionTextureCoordinate_3 {class=MaterialExpressionTextureCoordinate}
  MaterialExpressionTextureCoordinate_4 {class=MaterialExpressionTextureCoordinate}
  MaterialExpressionTextureCoordinate_5 {class=MaterialExpressionTextureCoordinate}
  MaterialExpressionTextureCoordinate_6 {class=MaterialExpressionTextureCoordinate}
  MaterialExpressionTextureCoordinate_7 {class=MaterialExpressionTextureCoordinate}
  MaterialExpressionTextureCoordinate_8 {class=MaterialExpressionTextureCoordinate}
  MaterialExpressionTextureSampleParameter2D_0 {
   class=MaterialExpressionTextureSampleParameter2D
   inputs [
    UVs {from=MaterialExpressionPanner_0, output=""}
    Apply View MipBias {from="", output=""}
   ]
   properties {parameterName=Alpha, texture=&/T_Contrail_Tandem_FilmRibbon}
  }
  MaterialExpressionTextureSample_1 {
   class=MaterialExpressionTextureSample
   inputs [
    UVs {from=MaterialExpressionAdd_0}
    Tex {from=""}
    Apply View MipBias {from=""}
   ]
   properties {samplerType=SAMPLERTYPE_Grayscale, texture=&LowResBlurredNoise}
  }
  MaterialExpressionTextureSample_2 {
   class=MaterialExpressionTextureSample
   inputs [
    UVs {from=MaterialExpressionTextureCoordinate_2}
    Tex {from=""}
    Apply View MipBias {from=""}
   ]
   properties {samplerType=SAMPLERTYPE_Grayscale, texture=&LowResBlurredNoise}
  }
  MaterialExpressionTime_0 {class=MaterialExpressionTime}
  MaterialExpressionTime_1 {class=MaterialExpressionTime}
  MaterialExpressionTime_3 {class=MaterialExpressionTime}
  MaterialExpressionTransform_0 {
   class=MaterialExpressionTransform
   inputs [None {from=MaterialExpressionVectorParameter_1, output=RGB}]
   properties {transformSourceType=TRANSFORMSOURCE_Local}
  }
  MaterialExpressionTransform_1 {
   class=MaterialExpressionTransform
   inputs [None {from=MaterialExpressionVectorParameter_2, output=RGB}]
   properties {transformSourceType=TRANSFORMSOURCE_Local}
  }
  MaterialExpressionVectorParameter_0 {class=MaterialExpressionVectorParameter, properties {parameterName=NoiseOffset, defaultValue=<rgba 0.2 0.3 0 0>}}
  MaterialExpressionVectorParameter_1 {class=MaterialExpressionVectorParameter, properties {parameterName=WPODirection, defaultValue=<rgba 0 0 15 0>}}
  MaterialExpressionVectorParameter_2 {class=MaterialExpressionVectorParameter, properties {parameterName=WPODirection2, defaultValue=<rgba 0 15 0 0>}}
  MaterialExpressionVectorParameter_3 {class=MaterialExpressionVectorParameter, properties {parameterName=Speed, defaultValue=<rgba -0.3 0 0 0>}}
  MaterialExpressionVectorParameter_4 {class=MaterialExpressionVectorParameter, properties {parameterName=ColorOffset}}
 }
 properties {
  bAutomaticallySetUsageInEditor=false
  bUsedWithNiagaraMeshParticles=true
  bUsedWithNiagaraRibbons=true
  bUsedWithNiagaraSprites=true
  blendMode=BLEND_Translucent
  shadingModel=MSM_Unlit
  twoSided=true
  usedShadingModels=Unlit
 }
}
