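# Masked, Default Lit material for Niagara mesh particles in the Tandem skydiving contrail cosmetic effect.
# Its graph multiplies particle color RGB by a BaseColor vector parameter, white by default, then scales the result down by a small constant.
# No connection from that chain to a material output appears, and no opacity mask logic is present.
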Material /BRCosmetics/Athena/Effects/Particles/Skydiving/NS_Contrail_Tandem/Materials/M_Tem_materials
graph {
 properties {bAutomaticallySetUsageInEditor=false, bUsedWithNiagaraMeshParticles=true, blendMode=BLEND_Masked, usedShadingModels=Default Lit}
 nodes [Multiply(A=Multiply(A=ParticleColor().RGB, B=VectorParameter() {parameterName=BaseColor, defaultValue=<rgba 1 1 1 0>}.RGB)) {constB=0.1}]
}
